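# MF_GradientBlend is a material function that turns a scalar gradient and a blend amount into a contrast-adjustable blend mask.
# GradientBase, previewed with texture U, is saturated, shaped by MF_Gain, expanded around its midpoint, then offset by GradientBias.
# The Blend input, adjusted by bias and multiplier parameters, is added to shift the gradient threshold.
# The sum is saturated and passed through CheapContrast, with tuning parameters grouped under GradientBlend.
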MaterialFunction /Game/Global/MaterialFunctions/MF_GradientBlend
graph {nodes [
  n1 = FunctionInput() {bUsePreviewValueAsDefault=true, inputName=Blend, inputType=FunctionInput_Scalar, sortPriority=2}
  n2 = ScalarParameter() {group=GradientBlend, parameterName=GBlend_BlendBias}
  n3 = FunctionInput(Preview=ComponentMask(None=TextureCoordinate()) {r=true}) {bUsePreviewValueAsDefault=true, inputName=GradientBase, inputType=FunctionInput_Scalar}
  n4 = ScalarParameter() {group=GradientBlend, parameterName=GradientExpansion}
  n5 = Multiply(
   A=MaterialFunctionCall!(value=Saturate(None=@n3), gain=ScalarParameter() {defaultValue=0.5, group=GradientBlend, parameterName=GBlend_Gain}) {materialFunction=&/MF_Gain}.Result
   B=Add(A=@n4)
  )
  n6 = Multiply(A=@n4) {constB=0.5}
  n7 = FunctionInput() {bUsePreviewValueAsDefault=true, inputName=GradientBias, inputType=FunctionInput_Scalar, sortPriority=1}
  n8 = Add(
   A=Multiply(A=Add(A=@n1, B=@n2), B=ScalarParameter() {defaultValue=1, group=GradientBlend, parameterName=GBlend_BlendMultiplier})
   B=Subtract(A=Subtract(A=@n5, B=@n6), B=Add(A=@n7) {constB=0.5}) {constB=0.5}
  )
  FunctionOutput(None=MaterialFunctionCall(In=Saturate(None=@n8), Contrast=ScalarParameter() {group=GradientBlend, parameterName=GBlend_Contrast}) {materialFunction=&CheapContrast}.Result)
 ]}
