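# M_GradientBlend is an unlit Material Gym material that exercises the MF_GradientBlend material function with animated inputs.
# The horizontal UV coordinate supplies the gradient base, and a sine of time scaled by Speed drives its bias.
# A tiled, scrollable texture, defaulting to atmospheric cloud noise, feeds the blend input through a selectable channel mask.
# Speed and scroll speed default to zero, so the animation is static until those parameters are overridden.
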Material /Game/Global/Materials/MaterialGym/Materials/GradientBlend/M_GradientBlend
graph {
 properties {bHasPixelAnimation=true, pixelDepthOffsetMode=PDOM_Legacy, shadingModel=MSM_Unlit, usedShadingModels=Unlit}
 nodes [
  n1 = Multiply(A=Time(), B=ScalarParameter() {parameterName=Speed})
  n2 = VectorParameter() {parameterName=TextureScrollSpeed, defaultValue=<rgba 0 0 0 1>}
  n3 = Add(
   A=Multiply(A=TextureCoordinate(), B=ScalarParameter() {defaultValue=1, parameterName=TextureTile})
   B=Multiply(A=ComponentMask(None=@n2.RGB) {g=true, r=true}, B=Time())
  )
  MaterialFunctionCall!(
   GradientBase=ComponentMask(None=TextureCoordinate()) {r=true}
   GradientBias=Multiply(A=Sine(None=@n1))
   Blend=ChannelMaskParameter(None=TextureSampleParameter2D(UVs=@n3) {parameterName=Texture, texture=&AtmosphericCloudNoise01}.RGBA) {parameterName=TextureChannel}.R
  ) {materialFunction=/Game/Global/MaterialFunctions/MF_GradientBlend}
 ]
}
